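# Physics asset for the Cape_M_MED_RoundThumb_Nail cosmetic cape, defining its dynamic bone simulation.
# Center, left and right chains of seven capsule bodies hang from MSTR roots, with tight twist limits and narrower swing near the top.
# Angularly free cross-constraints tie the side chains to the center at segments 05 and 07, and to each other at 07.
# Oversized upper arm, forearm and calf capsules provide body shapes around the cape.
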PhysicsAsset /BRCosmetics/Accessories/FORT_Capes/Cape_M_MED_RoundThumb_Nail/Meshes/M_MED_RoundThumb_Nail_Cape_Dyn
bodies {
 calf_l [{
   shapeType=Capsule
   length=35.8004
   radius=17.8996
   shapeName=calf_l_capsule
   center=<-21.6321 1.8098 1.7968>
   rotation=<rot 86.6447 0.0001 0.0002>
  }]
 calf_r [{
   shapeType=Capsule
   length=35.8004
   radius=17.8996
   shapeName=calf_r_capsule
   center=<22.0288 -1.8098 -1.82>
   rotation=<rot 86.6447 0.0001 0.0002>
  }]
 dyn_cape_01 [{shapeType=Capsule, length=7.6298, radius=10.1031, shapeName=dyn_cape_01_capsule, center=<7.15 0 8>, rotation=<rot 90 0 0>}]
 dyn_cape_02 [{shapeType=Capsule, length=13.347, radius=9.4801, shapeName=dyn_cape_01_capsule, center=<7.15 0 5>, rotation=<rot 90 0 -90>}]
 dyn_cape_03 [{shapeType=Capsule, length=11.3411, radius=9.3311, shapeName=dyn_cape_01_capsule, center=<7.15 0 5>, rotation=<rot 90 0 -90>}]
 dyn_cape_04 [{shapeType=Capsule, length=9.602, radius=8.1308, shapeName=dyn_cape_01_capsule, center=<7.15 0 5>, rotation=<rot 90 0 0>}]
 dyn_cape_05 [{shapeType=Capsule, length=9.602, radius=8.1308, shapeName=dyn_cape_01_capsule, center=<7.15 0 5>, rotation=<rot 90 0 0>}]
 dyn_cape_06 [{shapeType=Capsule, length=16.0709, radius=9.3701, shapeName=dyn_cape_01_capsule, center=<7.15 0 5>, rotation=<rot 90 0 -90>}]
 dyn_cape_07 [{shapeType=Capsule, length=13.1176, radius=12.4284, shapeName=dyn_cape_01_capsule, center=<5.8445 0 5>, rotation=<rot 90 0 -90>}]
 dyn_cape_MSTR [{shapeType=Capsule, length=0.505, radius=0.505, shapeName=dyn_cape_r_MSTR_capsule, rotation=<rot 0 -90 0>}]
 dyn_cape_l_01 [
  {shapeType=Capsule, length=10.9334, radius=8.7773, shapeName=dyn_cape_01_capsule, center=<7.15 0 8>, rotation=<rot 90 0 -90>}
  {
   shapeType=Capsule
   length=12.7377
   radius=8.1529
   shapeName=dyn_cape_l_01_capsule
   center=<11.0188 11.0275 4.8469>
   rotation=<rot 83.2377 0 0>
  }
 ]
 dyn_cape_l_02 [{shapeType=Capsule, length=13.347, radius=9.4801, shapeName=dyn_cape_01_capsule, center=<7.15 0 5>, rotation=<rot 90 0 -90>}]
 dyn_cape_l_03 [{shapeType=Capsule, length=11.3411, radius=9.3311, shapeName=dyn_cape_01_capsule, center=<7.15 0 5>, rotation=<rot 90 0 -90>}]
 dyn_cape_l_04 [{shapeType=Capsule, length=9.602, radius=8.1308, shapeName=dyn_cape_01_capsule, center=<7.15 0 5>, rotation=<rot 90 0 0>}]
 dyn_cape_l_05 [{shapeType=Capsule, length=9.602, radius=8.1308, shapeName=dyn_cape_01_capsule, center=<7.15 0 5>, rotation=<rot 90 0 0>}]
 dyn_cape_l_06 [{shapeType=Capsule, length=30.7147, radius=10.0473, shapeName=dyn_cape_01_capsule, center=<7.15 0 5>, rotation=<rot 90 0 -90>}]
 dyn_cape_l_07 [{shapeType=Capsule, length=27.7614, radius=13.1056, shapeName=dyn_cape_01_capsule, center=<5.8445 0 5>, rotation=<rot 90 0 -90>}]
 dyn_cape_l_MSTR [{shapeType=Capsule, length=0.505, radius=0.505, shapeName=dyn_cape_r_MSTR_capsule, rotation=<rot 0 -90 0>}]
 dyn_cape_r_01 [{shapeType=Capsule, length=10.9334, radius=8.7773, shapeName=dyn_cape_01_capsule, center=<7.15 0 8>, rotation=<rot 90 0 -90>}]
 dyn_cape_r_02 [{shapeType=Capsule, length=13.347, radius=9.4801, shapeName=dyn_cape_01_capsule, center=<7.15 0 5>, rotation=<rot 90 0 -90>}]
 dyn_cape_r_03 [{shapeType=Capsule, length=11.3411, radius=9.3311, shapeName=dyn_cape_01_capsule, center=<7.15 0 5>, rotation=<rot 90 0 -90>}]
 dyn_cape_r_04 [{shapeType=Capsule, length=9.602, radius=8.1308, shapeName=dyn_cape_01_capsule, center=<7.15 0 5>, rotation=<rot 90 0 0>}]
 dyn_cape_r_05 [{shapeType=Capsule, length=9.602, radius=8.1308, shapeName=dyn_cape_01_capsule, center=<7.15 0 5>, rotation=<rot 90 0 0>}]
 dyn_cape_r_06 [{shapeType=Capsule, length=30.7147, radius=10.0473, shapeName=dyn_cape_01_capsule, center=<7.15 0 5>, rotation=<rot 90 0 -90>}]
 dyn_cape_r_07 [{shapeType=Capsule, length=27.7614, radius=13.1056, shapeName=dyn_cape_01_capsule, center=<5.8445 0 5>, rotation=<rot 90 0 -90>}]
 dyn_cape_r_MSTR [{shapeType=Capsule, length=0.505, radius=0.505, shapeName=dyn_cape_r_MSTR_capsule, rotation=<rot 0 -90 0>}]
 lowerarm_l [{
   shapeType=Capsule
   length=15
   radius=15
   shapeName=lowerarm_l_capsule
   center=<23.2313 -0.0353 -0.9405>
   rotation=<rot 87.6542 8.6837 8.7636>
  }]
 lowerarm_r [{
   shapeType=Capsule
   length=15
   radius=15
   shapeName=lowerarm_r_capsule
   center=<-20.4137 -0.0393 1.3545>
   rotation=<rot 86.1854 -5.5514 -5.6494>
  }]
 upperarm_l [{shapeType=Capsule, length=16.446, radius=13.8039, shapeName=upperarm_r_capsule, center=<13.4724 0 0>, rotation=<rot -0.0001 -89.9999 90>}]
 upperarm_r [{shapeType=Capsule, length=18.1, radius=12.15, shapeName=upperarm_r_capsule, center=<-13.4724 0 0>, rotation=<rot -0.0001 -89.9999 -90>}]
}
constraints [
 {bone1Name=dyn_cape_r_01, bone2Name=dyn_cape_r_MSTR, swing1LimitDegrees=10, swing2LimitDegrees=45, twistLimitDegrees=5}
 {bone1Name=dyn_cape_r_02, bone2Name=dyn_cape_r_01, swing1LimitDegrees=20, swing2LimitDegrees=45, twistLimitDegrees=5}
 {bone1Name=dyn_cape_r_03, bone2Name=dyn_cape_r_02, swing1LimitDegrees=45, swing2LimitDegrees=45, twistLimitDegrees=5}
 {bone1Name=dyn_cape_r_04, bone2Name=dyn_cape_r_03, swing1LimitDegrees=45, swing2LimitDegrees=45, twistLimitDegrees=5}
 {bone1Name=dyn_cape_r_05, bone2Name=dyn_cape_r_04, swing1LimitDegrees=45, swing2LimitDegrees=45, twistLimitDegrees=5}
 {bone1Name=dyn_cape_r_06, bone2Name=dyn_cape_r_05, swing1LimitDegrees=45, swing2LimitDegrees=45, twistLimitDegrees=5}
 {bone1Name=dyn_cape_r_07, bone2Name=dyn_cape_r_06, swing1LimitDegrees=45, swing2LimitDegrees=62.28, twistLimitDegrees=5}
 {bone1Name=dyn_cape_l_01, bone2Name=dyn_cape_l_MSTR, swing1LimitDegrees=10, swing2LimitDegrees=45, twistLimitDegrees=5}
 {bone1Name=dyn_cape_l_02, bone2Name=dyn_cape_l_01, swing1LimitDegrees=20, swing2LimitDegrees=45, twistLimitDegrees=5}
 {bone1Name=dyn_cape_l_03, bone2Name=dyn_cape_l_02, swing1LimitDegrees=45, swing2LimitDegrees=45, twistLimitDegrees=5}
 {bone1Name=dyn_cape_l_04, bone2Name=dyn_cape_l_03, swing1LimitDegrees=45, swing2LimitDegrees=45, twistLimitDegrees=5}
 {bone1Name=dyn_cape_l_05, bone2Name=dyn_cape_l_04, swing1LimitDegrees=45, swing2LimitDegrees=45, twistLimitDegrees=5}
 {bone1Name=dyn_cape_l_06, bone2Name=dyn_cape_l_05, swing1LimitDegrees=45, swing2LimitDegrees=45, twistLimitDegrees=5}
 {bone1Name=dyn_cape_l_07, bone2Name=dyn_cape_l_06, swing1LimitDegrees=45, swing2LimitDegrees=45, twistLimitDegrees=5}
 {bone1Name=dyn_cape_01, bone2Name=dyn_cape_MSTR, swing1LimitDegrees=10, swing2LimitDegrees=45, twistLimitDegrees=5}
 {bone1Name=dyn_cape_02, bone2Name=dyn_cape_01, swing1LimitDegrees=20, swing2LimitDegrees=45, twistLimitDegrees=5}
 {bone1Name=dyn_cape_03, bone2Name=dyn_cape_02, swing1LimitDegrees=45, swing2LimitDegrees=45, twistLimitDegrees=5}
 {bone1Name=dyn_cape_04, bone2Name=dyn_cape_03, swing1LimitDegrees=45, swing2LimitDegrees=45, twistLimitDegrees=5}
 {bone1Name=dyn_cape_05, bone2Name=dyn_cape_04, swing1LimitDegrees=45, swing2LimitDegrees=45, twistLimitDegrees=5}
 {bone1Name=dyn_cape_06, bone2Name=dyn_cape_05, swing1LimitDegrees=45, swing2LimitDegrees=45, twistLimitDegrees=5}
 {bone1Name=dyn_cape_07, bone2Name=dyn_cape_06, swing1LimitDegrees=45, swing2LimitDegrees=45, twistLimitDegrees=5}
 {
  bone1Name=dyn_cape_r_05
  bone2Name=dyn_cape_05
  swing1LimitDegrees=45
  swing1Motion=Free
  swing2LimitDegrees=45
  swing2Motion=Free
  twistLimitDegrees=45
  twistMotion=Free
 }
 {
  bone1Name=dyn_cape_l_05
  bone2Name=dyn_cape_05
  swing1LimitDegrees=45
  swing1Motion=Free
  swing2LimitDegrees=45
  swing2Motion=Free
  twistLimitDegrees=45
  twistMotion=Free
 }
 {
  bone1Name=dyn_cape_l_07
  bone2Name=dyn_cape_07
  swing1LimitDegrees=45
  swing1Motion=Free
  swing2LimitDegrees=45
  swing2Motion=Free
  twistLimitDegrees=45
  twistMotion=Free
 }
 {
  bone1Name=dyn_cape_r_07
  bone2Name=dyn_cape_07
  swing1LimitDegrees=45
  swing1Motion=Free
  swing2LimitDegrees=45
  swing2Motion=Free
  twistLimitDegrees=45
  twistMotion=Free
 }
 {
  bone1Name=dyn_cape_l_07
  bone2Name=dyn_cape_r_07
  swing1LimitDegrees=45
  swing1Motion=Free
  swing2LimitDegrees=45
  swing2Motion=Free
  twistLimitDegrees=45
  twistMotion=Free
 }
 {
  bone1Name=calf_r
  bone2Name=calf_l
  swing1LimitDegrees=45
  swing1Motion=Free
  swing2LimitDegrees=45
  swing2Motion=Free
  twistLimitDegrees=45
  twistMotion=Free
 }
]
